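# Nanite-enabled material for the WaddleJam dance floor, with animated glowing logo and duck footprints.
# World-position UVs relative to a floor origin parameter drive sine-rotated, drifting logo samples, while tiled footprints are hue-shifted and blink using per-texel phase offsets.
# Both layers combine into pulsing emissive over dark base color and PBR values masked by UV1, with quality switches simplifying lower tiers.
# Temporal responsiveness is set, and a dance-floor texture sample sits unconnected.
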Material /Hera/Sets/WaddleJam/TechArt/M_WaddleDanceFloor
graph {
 properties {
  bAutomaticallySetUsageInEditor=false
  bUseMaterialAttributes=true
  bUsedWithNanite=true
  previewMesh=/Game/Effects/Fort_Effects/Effects/Environment/Hera/Shards/Glow/SM_Env_TeaCake_Shard_Glows
  usedShadingModels=Default Lit
 }
 nodes [
  MaterialFunctionCall!(Sine Phase=Time(), Value 1=Constant(), Value 2=Constant() {r=1}) {materialFunction=/Engine/Functions/Engine_MaterialFunctions02/Utility/Sine_Remapped}
  Multiply(
   A=TextureCoordinate() {coordinateIndex=2}
   B=NamedRerouteDeclaration(None=ScalarParameter() {defaultValue=1, group=Patterns, parameterName=Dancefloor_Lights_Tile}) {name=PatternTile, nodeColor=<rgba 1 0 0.5391 1>}
  )
  n1 = Time()
  Multiply(A=@n1) {constB=4}
  n2 = TextureCoordinate() {coordinateIndex=1}
  NamedRerouteDeclaration(None=Floor(None=ComponentMask(None=@n2) {r=true})) {name=TexCordUV1 Mask, nodeColor=<rgba 0.8281 0 1 1>}
  n3 = NamedRerouteUsage()
  n4 = LinearInterpolate(Alpha=@n3) {constA=0.5, constB=0}
  n5 = Constant()
  n6 = Constant() {r=0.5}
  n7 = Constant()
  NamedRerouteDeclaration(None=SetMaterialAttributes!(
    MaterialAttributes
    Metallic=QualitySwitch(Default=@n4, Low=@n5, High=@n4, Medium=@n5, Epic=@n4) {desc=Metal}
    Roughness=LinearInterpolate(Alpha=@n3) {constA=0.1, desc=Rough}
    Specular=QualitySwitch(Default=@n6, Low=@n7, High=@n6, Medium=@n7, Epic=@n6) {desc=Specular}
   ) {attributeSetTypes [#1, #2, #3]}) {name=PBR MA, nodeColor=<rgba 0 1 0.5078 1>}
  n8 = WorldPosition()
  n9 = ComponentMask(None=Subtract(A=TruncateLWC!(None=@n8.XYZ), B=VectorParameter() {parameterName=Dance Floor Position, defaultValue=<rgba -115199 72698 395 1>}.RGB)) {g=true, r=true}
  n10 = ScalarParameter() {defaultValue=950, parameterName=Tex Scale}
  n11 = VectorParameter() {parameterName=TEx Pos, defaultValue=<rgba 0.5 0.5 0 0>}
  NamedRerouteDeclaration(None=Add(A=Divide(A=@n9, B=@n10) {constB=125}, B=ComponentMask(None=@n11.RGB) {g=true, r=true})) {name=WP UVs, nodeColor=<rgba 1 0 0.1406 1>}
  n12 = NamedRerouteUsage()
  n13 = Multiply(A=Time(), B=ScalarParameter() {defaultValue=1.4, parameterName=Base Time Speed})
  n14 = MaterialFunctionCall(
   Sine Phase=Divide(A=@n13) {constB=26}
   Value 1=ScalarParameter() {defaultValue=0.5}
   Value 2=ScalarParameter() {defaultValue=-0.5, parameterName=Param_1}
  ) {materialFunction=&Sine_Remapped}
  n15 = MaterialFunctionCall(Sine Phase=Divide(A=@n13) {constB=28.5}, Value 1=Constant() {r=-0.041}, Value 2=Constant() {r=0.052}) {materialFunction=&Sine_Remapped}
  n16 = TextureSample(UVs=Add(A=Rotator(Coordinate=@n12, Time=@n14.Remapped Sine), B=AppendVector(A=@n15.Remapped Sine, B=@n15.Remapped Sine))) {samplerSource=SSM_Wrap_WorldGroupSettings, samplerType=SAMPLERTYPE_Masks, texture=&/T_WaddleLogo}
  n17 = Add(A=@n13, B=ScalarParameter() {defaultValue=3.7013, parameterName=Param_4})
  n18 = Rotator(
   Coordinate=NamedRerouteUsage()
   Time=MaterialFunctionCall(Sine Phase=Divide(A=@n17) {constB=25}, Value 1=Constant() {r=0.8}, Value 2=Constant() {r=-0.4}) {materialFunction=&Sine_Remapped}.Remapped Sine
  )
  n19 = MaterialFunctionCall(Sine Phase=Divide(A=@n17) {constB=32}, Value 1=Constant() {r=0.117}, Value 2=Constant() {r=-0.112}) {materialFunction=&Sine_Remapped}
  n20 = AppendVector(A=@n19.Remapped Sine, B=@n19.Remapped Sine)
  n21 = Multiply(A=TextureSample(UVs=Add(A=@n18, B=@n20)) {samplerSource=SSM_Wrap_WorldGroupSettings, samplerType=SAMPLERTYPE_Masks, texture=&T_WaddleLogo}.R) {constB=0.4}
  n22 = VectorParameter() {defaultValue=<rgba 0.3043 1 1 1>}
  n23 = LinearInterpolate(A=Add(A=@n16.R, B=Multiply(A=@n21, B=@n22.RGB)), B=@n16.R, Alpha=Add(A=@n14.Remapped Sine) {constB=0.5})
  n24 = TextureSample(UVs=NamedRerouteUsage()) {samplerSource=SSM_Wrap_WorldGroupSettings, samplerType=SAMPLERTYPE_Masks, texture=&T_WaddleLogo}
  NamedRerouteDeclaration(None=Multiply(A=Add(A=@n23, B=@n24.G), B=VectorParameter() {parameterName=Logo Color, defaultValue=<rgba 0.7023 0.0833 1 1>}.RGB)) {name=Logo, nodeColor=<rgba 0.9219 1 0 1>}
  n25 = TextureSample(UVs=Multiply(A=NamedRerouteUsage(), B=ScalarParameter() {defaultValue=1.1, parameterName=Duck Feet Tile})) {samplerSource=SSM_Wrap_WorldGroupSettings, samplerType=SAMPLERTYPE_LinearColor, texture=&/T_DuckFootprints}
  n26 = Multiply(A=@n25.G) {constB=0.2}
  n27 = MaterialFunctionCall(
   Hue Shift Percentage=@n25.B
   Texture=Multiply(A=Add(A=@n25.R, B=@n26), B=VectorParameter() {parameterName=Duck Feet Color, defaultValue=<rgba 0.2641 1.3 0.8049 1>}.RGB)
  ) {materialFunction=&HueShift}
  n28 = Divide(A=Time())
  n29 = Saturate(None=Sine(None=Add(A=@n25.B, B=@n28)))
  n30 = Time()
  n31 = Sine(None=Add(A=@n25.A, B=Divide(A=@n30) {constB=2.5}))
  n32 = Multiply(A=Multiply(A=@n25.A, B=Saturate(None=@n31))) {constB=0.5}
  n33 = VectorParameter() {parameterName=Logo Color, defaultValue=<rgba 0.7023 0.0833 1 1>}
  NamedRerouteDeclaration(None=Add(A=Multiply(A=@n27.Result, B=@n29), B=Multiply(A=@n32, B=@n33.RGB))) {name=Feets, nodeColor=<rgba 0.7578 0 1 1>}
  n34 = Constant() {r=1}
  n35 = Multiply(
   A=VectorParameter() {parameterName=Base Color, defaultValue=<rgba 0.0729 0.0729 0.0729 1>}.RGB
   B=QualitySwitch(Default=@n34, Low=Constant() {r=0.2}, High=@n34, Medium=@n34, Epic=@n34)
  ) {constA=0.25}
  n36 = VectorParameter() {parameterName=Base Color Bottom, defaultValue=<rgba 0.026 0.026 0.026 1>}
  n37 = NamedRerouteUsage()
  n38 = Multiply(
   A=Add(A=NamedRerouteUsage(), B=NamedRerouteUsage())
   B=MaterialFunctionCall(
    Sine Phase=@n1
    Value 1=ScalarParameter() {defaultValue=0.8, parameterName=Param_2}
    Value 2=ScalarParameter() {defaultValue=1.4, parameterName=Param_3}
   ) {materialFunction=&Sine_Remapped}.Remapped Sine
  )
  SetMaterialAttributes!(
   MaterialAttributes=SetMaterialAttributes!(MaterialAttributes=NamedRerouteUsage(), Base Color=LinearInterpolate(A=@n35, B=@n36.RGB, Alpha=@n37)) {attributeSetTypes [#4]}
   Emissive Color=Multiply(
    A=NamedRerouteDeclaration(None=@n38) {name=Emissive, nodeColor=<rgba 0 1 0.4375 1>}
    B=ScalarParameter() {defaultValue=1, parameterName=Emissive Intensity}
   )
  ) {attributeSetTypes [#5]}
  TemporalResponsivenessOutput!(None=Constant() {r=0.5})
  TextureSample(UVs=Multiply(A=TextureCoordinate(), B=NamedRerouteUsage())) {samplerSource=SSM_Wrap_WorldGroupSettings, texture=/Game/Environments/Sets/Concert/Textures/T_DanceFloor_D}
  Time()
 ]
}
